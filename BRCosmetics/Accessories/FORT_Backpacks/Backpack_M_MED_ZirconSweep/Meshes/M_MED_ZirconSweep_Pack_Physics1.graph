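# Physics asset for the ZirconSweep backpack cosmetic, defining simple collision bodies and joints for its pack bone and three "dyn" bones.
# Each bone has one capsule, with the middle dyn bone's being the largest and the pack's a tiny one.
# The left, middle and right dyn bodies each constrain directly to the pack in a star layout.
# All three constraints lock swing and twist, so their angular limit values have no effect.
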PhysicsAsset /BRCosmetics/Accessories/FORT_Backpacks/Backpack_M_MED_ZirconSweep/Meshes/M_MED_ZirconSweep_Pack_Physics1
bodies {
 dyn_l_joint [{shapeType=Capsule, length=2, radius=2, shapeName=dyn_l_joint_capsule, center=<0 0 0.2826>}]
 dyn_mid_joint [{
   shapeType=Capsule
   length=9.7201
   radius=3.0879
   shapeName=dyn_mid_joint_capsule
   center=<1.7415 -0.607 0.8607>
   rotation=<rot 0 -89.6893 -74.0875>
  }]
 dyn_r_joint [{shapeType=Capsule, length=2, radius=2, shapeName=dyn_r_joint_capsule}]
 pack [{shapeType=Capsule, length=0.505, radius=0.505, shapeName=pack_capsule, rotation=<rot 0 -90 0>}]
}
constraints [
 {
  bone1Name=dyn_r_joint
  bone2Name=pack
  swing1LimitDegrees=45
  swing1Motion=Locked
  swing2LimitDegrees=45
  swing2Motion=Locked
  twistLimitDegrees=45
  twistMotion=Locked
 }
 {
  bone1Name=dyn_l_joint
  bone2Name=pack
  swing1LimitDegrees=45
  swing1Motion=Locked
  swing2LimitDegrees=45
  swing2Motion=Locked
  twistLimitDegrees=45
  twistMotion=Locked
 }
 {
  bone1Name=dyn_mid_joint
  bone2Name=pack
  swing1LimitDegrees=45
  swing1Motion=Locked
  swing2LimitDegrees=45
  swing2Motion=Locked
  twistLimitDegrees=45
  twistMotion=Locked
 }
]
