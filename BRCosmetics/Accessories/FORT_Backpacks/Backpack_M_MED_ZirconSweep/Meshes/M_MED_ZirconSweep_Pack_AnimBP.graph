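# Animation Blueprint for the ZirconSweep backpack cosmetic, adding physics sway and a floating motion to the pack.
# A RigidBody node simulates relative to the pack bone, then additive bone-space translations offset the left, right and middle dyn joints.
# Each update builds those offsets from game-time sine waves at differing frequencies.
# It also swaps the physics gravity: normal usually, but when skydiving without parachute, upward if rising, otherwise opposite pelvis forward plus downward.
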AnimBlueprint /BRCosmetics/Accessories/FORT_Backpacks/Backpack_M_MED_ZirconSweep/Meshes/M_MED_ZirconSweep_Pack_AnimBP
target_skeleton=&/M_MED_ZirconSweep_Pack_Skeleton
graphs [
 AnimGraph {nodes [
   n1 = RigidBody!(
    ComponentPose=Local To Component(LocalPose=Input Pose())
    Alpha=0.98
    OverrideWorldGravity:Vector=Get Gvatity Override() -> (Gvatity Override:Vector)
    ExternalForce
   ) -> (Pose) {
    bOverrideWorldGravity=true
    overridePhysicsAsset=&/M_MED_ZirconSweep_Pack_Physics1
    simulationSpace=BaseBoneSpace
    alphaBoolBlend {blendOption=Linear}
    baseBoneRef {boneName=pack}
    componentAppliedLinearAccClamp=<1000 100 1000>
    componentLinearAccScale=<0.12 0.1 0.12>
    componentLinearVelScale=<0.1 0.1 0.1>
    simSpaceSettings {maxAngularAcceleration=30, maxAngularVelocity=1, maxLinearAcceleration=1315, maxLinearVelocity=9.8, velocityScaleZ=16, worldAlpha=0.75}
   }
   n2 = "Transform (Modify) Bone - Bone: dyn_l_joint"(ComponentPose=@n1, Alpha=1, Translation=Get dyn_l_pack_Float() -> (dyn_l_pack_Float:Vector), Rotation="-5,0,0", Scale="1,1,1") {
    rotationMode=BMM_Additive
    rotationSpace=BCS_BoneSpace
    translationMode=BMM_Additive
    translationSpace=BCS_BoneSpace
    alphaBoolBlend {blendOption=Linear}
    alphaScaleBiasClamp {bInterpResult=true, interpSpeedDecreasing=10000, interpSpeedIncreasing=10000}
    boneToModify {boneName=dyn_l_joint}
    rotation {pitch=-5}
   }
   n3 = "Transform (Modify) Bone - Bone: dyn_r_joint"(ComponentPose=@n2, Alpha=1, Translation=Get dyn_r_pack_Float() -> (dyn_r_pack_Float:Vector), Rotation="0,0,0", Scale="1,1,1") {
    rotationSpace=BCS_BoneSpace
    translationMode=BMM_Additive
    translationSpace=BCS_BoneSpace
    alphaBoolBlend {blendOption=Linear}
    boneToModify {boneName=dyn_r_joint}
   }
   n4 = "Transform (Modify) Bone - Bone: dyn_mid_joint"(ComponentPose=@n3, Alpha=1, Translation=Get dyn_mid_pack_Float_0() -> (dyn_mid_pack_Float_0:Vector), Rotation="0,0,0", Scale="1,1,1") {
    rotationSpace=BCS_BoneSpace
    translationMode=BMM_Additive
    translationSpace=BCS_BoneSpace
    alphaBoolBlend {blendOption=Linear}
    boneToModify {boneName=dyn_mid_joint}
   }
   Output Pose(Result=Component To Local(ComponentPose=@n4)) {layerGroup=None}
  ]}
 EventGraph {nodes [
   Event BlueprintUpdateAnimation()
   > n1 = Sequence!(execute) -> (then_0, then_1, then_2:Exec, then_3:Exec)
   n2 = GetGameTimeInSeconds()
   n3 = "Sin (Radians)"(A=float * float(A=1.2321, B=@n2))
   n4 = "Sin (Radians)"(A=float * float(A=1.3512, B=@n2))
   n5 = float * float(A="Sin (Radians)"(A=float * float(A=0.796, B=@n2)), B=0.9)
   > Set dyn_l_pack_Float(dyn_l_pack_Float:Vector=MakeVector(X=float * float(A=@n4, B=0.6), Y=float * float(A=@n3, B=0.3), Z=float + float(A=@n5, B=0.05))) -> (Output_Get:Vector)
   n6 = Branch(execute=@n1.then_3, Condition=Get bIsSkyDiving())
   > n7 = Branch(Condition=Get bIsParachuteOpened())
   Set Gvatity Override(execute=[@n6.else, @n7.then], Gvatity Override:Vector="0,0,-980") -> (Output_Get:Vector)
   n8 = Branch(execute=@n7.else, Condition="float > float"(A=Get PawnVelocityZ(), B=50))
   > Set Gvatity Override(Gvatity Override="0,0,200")
   n9 = TryGetPawnOwner()
   n10 = GetForwardVector!(InRot:Rotator=GetSocketRotation!(
     self:Scene Component Object Reference=Get RootComponent!(self:Actor Object Reference=@n9) -> (RootComponent:Scene Component Object Reference)
     InSocketName:Name=pelvis
    ) -> (ReturnValue:Rotator)) -> (ReturnValue:Vector)
   Set Gvatity Override(execute=@n8.else, Gvatity Override=vector + vector(A=vector * vector(A=@n10, B:float=-1000) -> (ReturnValue=~), B="0,0,-200"))
   n11 = "Sin (Radians)"(A=float * float(A=1.2321, B=@n2))
   n12 = float * float(A="Sin (Radians)"(A=float * float(A=1.352, B=@n2)), B=0.3)
   n13 = float * float(A="Sin (Radians)"(A=float * float(A=1.8, B=@n2)), B=0.7)
   Set dyn_r_pack_Float(
    execute=@n1.then_1
    dyn_r_pack_Float:Vector=MakeVector(X=float + float(A=@n12, B=0.7), Y=float * float(A=@n11, B=0.3), Z=float + float(A=@n13, B=0.3))
   ) -> (Output_Get:Vector)
   n14 = "Sin (Radians)"(A=float * float(A=1.2321, B=@n2))
   n15 = "Sin (Radians)"(A=float * float(A=1.36, B=@n2))
   n16 = "Sin (Radians)"(A=float * float(A=1.71, B=@n2))
   Set dyn_mid_pack_Float_0(
    execute=@n1.then_2
    dyn_mid_pack_Float_0:Vector=MakeVector(X=float * float(A=@n15, B=1.12), Y=float * float(A=@n14, B=0.3), Z=float * float(A=@n16, B=0.17))
   ) -> (Output_Get:Vector)
   GetRealTimeSeconds()
   GetRealTimeSeconds()
   GetRealTimeSeconds()
  ]}
]
variables [
 Active Child Index {type=integer}
 Duration {type=number, default=0.1}
 dyn_l_pack_Float {type=Vector, default=<0.9698 0.0175 0.0535>}
 dyn_mid_pack_Float_0 {type=Vector, default=<0.8973 0.0175 -0.3869>}
 dyn_r_pack_Float {type=Vector, default=<0.8817 0.0175 -0.0583>}
 Gvatity Override {type=Vector}
 Linear Acc Scale {type=Vector, default=<0.1 0.1 0.1>}
 Linear Vel Scale {type=Vector, default=<0 0 1.3>}
 Range {
  type=SequencerScriptingRange
  default {bHasEnd=false, bHasStart=false, exclusiveEnd=0, inclusiveStart=0, internalRate {denominator=1, numerator=60000}}
 }
 Row Name {type=string, default=None}
]
